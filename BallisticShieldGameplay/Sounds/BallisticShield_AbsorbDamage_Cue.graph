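# Sound cue for the ballistic shield absorbing damage, layering stagger build-up one-shots with melee impacts.
# isLocallyViewed selects between instigator hit-marker and standard shield attenuation, both feeding one isMeleeHit branch.
# Melee hits mix random impacts and a sweetener over the build-up layer; other hits play only that layer.
# A continuous Stagger parameter drives build-up pitch and volume, while quality level picks its ten- or three-wave random pool.
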SoundCue /BallisticShieldGameplay/Sounds/BallisticShield_AbsorbDamage_Cue
root=Branch_1
nodes [
 {
  id=Attenuation_0
  children [Branch_0]
  properties {attenuationSettings=&/BallisticShield_HitMarker_Instigate_Att}
 }
 {id=Attenuation_1, children [Branch_0], properties {attenuationSettings=&/BallisticShield_Att}}
 {id=Branch_0, children [Mixer_2,Mixer_1,Mixer_1], properties {boolParameterName=isMeleeHit}}
 {id=Branch_1, children [Attenuation_0, Attenuation_1, Attenuation_1], properties {boolParameterName=isLocallyViewed}}
 {id=Mixer_1, children [ModulatorContinuous_0, null], properties {inputVolume [1.95,0.1]}}
 {id=Mixer_2, children [Mixer_3, ModulatorContinuous_0], properties {inputVolume [0.5,1]}}
 {id=Mixer_3, children [Modulator_0,Modulator_1], properties {inputVolume [0.5,0.7]}}
 {
  id=ModulatorContinuous_0
  children [Modulator_3]
  properties {
   pitchModulationParams {default=0, maxOutput=0.9, minOutput=0.7, parameterName=Stagger}
   volumeModulationParams {default=0, minInput=0.3, parameterName=Stagger}
  }
 }
 {id=Modulator_0, children [Random_4], properties {pitchMax=1.3, pitchMin=1.2}}
 {id=Modulator_1, children [WavePlayer_30], properties {pitchMax=1.3, pitchMin=1.2}}
 {id=Modulator_3, children [QualityLevel_1], properties {pitchMax=1.3, pitchMin=1.2}}
 {id=QualityLevel_1, children [Random_1,Random_1,Random_3,Random_3]}
 {
  id=Random_1
  children [
   WavePlayer_6
   WavePlayer_7
   WavePlayer_8
   WavePlayer_9
   WavePlayer_16
   WavePlayer_17
   WavePlayer_18
   WavePlayer_19
   WavePlayer_20
   WavePlayer_21
  ]
  properties {weights [1,1,1,1,1,1,1,1,1,1]}
 }
 {id=Random_3, children [WavePlayer_19, WavePlayer_20, WavePlayer_21], properties {bShouldExcludeFromBranchCulling=true, weights [1,1,1]}}
 {id=Random_4, children [WavePlayer_26, WavePlayer_27, WavePlayer_28, WavePlayer_29], properties {weights [1,1,1,1]}}
 {id=WavePlayer_16, properties {soundWaveAssetPtr=&/Paprika_BallisticShield_StaggerBuildUp_OneShot_05}}
 {id=WavePlayer_17, properties {soundWaveAssetPtr=&/Paprika_BallisticShield_StaggerBuildUp_OneShot_06}}
 {id=WavePlayer_18, properties {soundWaveAssetPtr=&/Paprika_BallisticShield_StaggerBuildUp_OneShot_07}}
 {id=WavePlayer_19, properties {soundWaveAssetPtr=&/Paprika_BallisticShield_StaggerBuildUp_OneShot_08}}
 {id=WavePlayer_20, properties {soundWaveAssetPtr=&/Paprika_BallisticShield_StaggerBuildUp_OneShot_09}}
 {id=WavePlayer_21, properties {soundWaveAssetPtr=&/Paprika_BallisticShield_StaggerBuildUp_OneShot_10}}
 {id=WavePlayer_26, properties {soundWaveAssetPtr=&/Paprika_BallisticShield_MeleeImpact_01}}
 {id=WavePlayer_27, properties {soundWaveAssetPtr=&/Paprika_BallisticShield_MeleeImpact_02}}
 {id=WavePlayer_28, properties {soundWaveAssetPtr=&/Paprika_BallisticShield_MeleeImpact_03}}
 {id=WavePlayer_29, properties {soundWaveAssetPtr=&/Paprika_BallisticShield_MeleeImpact_04}}
 {id=WavePlayer_30, properties {soundWaveAssetPtr=&/Paprika_BallisticShield_MeleeImpact_Sweetner}}
 {id=WavePlayer_6, properties {soundWaveAssetPtr=&/Paprika_BallisticShield_StaggerBuildUp_OneShot_01}}
 {id=WavePlayer_7, properties {soundWaveAssetPtr=&/Paprika_BallisticShield_StaggerBuildUp_OneShot_02}}
 {id=WavePlayer_8, properties {soundWaveAssetPtr=&/Paprika_BallisticShield_StaggerBuildUp_OneShot_03}}
 {id=WavePlayer_9, properties {soundWaveAssetPtr=&/Paprika_BallisticShield_StaggerBuildUp_OneShot_04}}
]
properties {maxDistance=2097152, soundClassObject=&Fort_WeaponMelee, volumeMultiplier=1}
