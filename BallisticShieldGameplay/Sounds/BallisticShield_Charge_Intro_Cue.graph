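# Sound cue for the intro of the ballistic shield charge, layering a whoosh with a max-charge one-shot.
# A mixer plays a random pick of four whooshes at reduced volume alongside a boosted random pick of six one-shots, delayed by a fixed quarter second.
# It uses the ballistic shield attenuation and melee weapon sound class, with a concurrency override limiting it to one instance.
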SoundCue /BallisticShieldGameplay/Sounds/BallisticShield_Charge_Intro_Cue
root=Mixer_0
nodes [
 {id=Delay_0, children [Random_1], properties {delayMax=0.25, delayMin=0.25}}
 {id=Mixer_0, children [Random_0,Delay_0], properties {inputVolume [0.25,2]}}
 {id=Random_0, children [WavePlayer_0, WavePlayer_9, WavePlayer_10, WavePlayer_11], properties {weights [1,1,1,1]}}
 {id=Random_1, children [WavePlayer_1,WavePlayer_2,WavePlayer_3,WavePlayer_4,WavePlayer_5,WavePlayer_6], properties {weights [1,1,1,1,1,1]}}
 {id=WavePlayer_0, properties {soundWaveAssetPtr=&/Paprika_BallisticShield_Charge_Whoosh_04}}
 {id=WavePlayer_1, properties {soundWaveAssetPtr=&/Paprika_BallisticShield_MaxChargeOneShot_01}}
 {id=WavePlayer_10, properties {soundWaveAssetPtr=&/Paprika_BallisticShield_Charge_Whoosh_02}}
 {id=WavePlayer_11, properties {soundWaveAssetPtr=&/Paprika_BallisticShield_Charge_Whoosh_01}}
 {id=WavePlayer_2, properties {soundWaveAssetPtr=&/Paprika_BallisticShield_MaxChargeOneShot_02}}
 {id=WavePlayer_3, properties {soundWaveAssetPtr=&/Paprika_BallisticShield_MaxChargeOneShot_03}}
 {id=WavePlayer_4, properties {soundWaveAssetPtr=&/Paprika_BallisticShield_MaxChargeOneShot_04}}
 {id=WavePlayer_5, properties {soundWaveAssetPtr=&/Paprika_BallisticShield_MaxChargeOneShot_05}}
 {id=WavePlayer_6, properties {soundWaveAssetPtr=&/Paprika_BallisticShield_MaxChargeOneShot_06}}
 {id=WavePlayer_9, properties {soundWaveAssetPtr=&/Paprika_BallisticShield_Charge_Whoosh_03}}
]
properties {
 attenuationSettings=&/BallisticShield_Att
 bOverrideConcurrency=true
 maxDistance=3400
 soundClassObject=&Fort_WeaponMelee
 volumeMultiplier=1
 concurrencyOverrides {maxCount=1}
}
