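# Sound cue for the ballistic shield's charge release, played as an attenuated melee weapon sound limited to one concurrent instance.
# A mixer layers a reused AMR beam loop-end sound, a random charge-empty variant and a briefly delayed random charge-stop variant, with the stop layer boosted well above the others.
# An isFriendly branch plays that mix at half volume for friendlies and at full volume when false or unset.
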SoundCue /BallisticShieldGameplay/Sounds/BallisticShield_Charge_Release_Cue
root=Branch_0
nodes [
 {id=Branch_0, children [Modulator_1,Modulator_0,Modulator_0], properties {boolParameterName=isFriendly}}
 {id=Delay_0, children [Random_2], properties {delayMax=0.1, delayMin=0.1}}
 {id=Mixer_0, children [WavePlayer_5,Random_1,Delay_0], properties {inputVolume [0.2,0.3,2]}}
 {id=Modulator_0, children [Mixer_0], properties {volumeMax=1, volumeMin=1}}
 {id=Modulator_1, children [Mixer_0], properties {volumeMax=0.5, volumeMin=0.5}}
 {id=Random_1, children [WavePlayer_1,WavePlayer_2,WavePlayer_3], properties {weights [1,1,1]}}
 {id=Random_2, children [WavePlayer_4,WavePlayer_6,WavePlayer_7,WavePlayer_8], properties {weights [1,1,1,1]}}
 {id=WavePlayer_1, properties {soundWaveAssetPtr=&/Paprika_BallisticShield_Charge_Empty_01}}
 {id=WavePlayer_2, properties {soundWaveAssetPtr=&/Paprika_BallisticShield_Charge_Empty_02}}
 {id=WavePlayer_3, properties {soundWaveAssetPtr=&/Paprika_BallisticShield_Charge_Empty_03}}
 {id=WavePlayer_4, properties {soundWaveAssetPtr=&/Paprika_BallisticShield_ChargeStop_01}}
 {id=WavePlayer_5, properties {soundWaveAssetPtr=/Game/Sounds/Fort_Weapons/AMR/AMR_Beam_Fire_Loop_End}}
 {id=WavePlayer_6, properties {soundWaveAssetPtr=&/Paprika_BallisticShield_ChargeStop_02}}
 {id=WavePlayer_7, properties {soundWaveAssetPtr=&/Paprika_BallisticShield_ChargeStop_03}}
 {id=WavePlayer_8, properties {soundWaveAssetPtr=&/Paprika_BallisticShield_ChargeStop_04}}
]
properties {
 attenuationSettings=&/BallisticShield_Att
 bOverrideConcurrency=true
 maxDistance=3400
 soundClassObject=&Fort_WeaponMelee
 volumeMultiplier=1
 concurrencyOverrides {maxCount=1}
}
