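# Sound cue for the impact of a dropped ballistic shield, using weapon-drop attenuation and the loot-collect sound class.
# A mixer layers a random sweetener, through a default modulator, with a random drop-land sound at fixed pitch.
# A quality-level node gives the drop-land layer either all six variations or a subset of three.
# Concurrency is capped at a few instances, preventing new plays beyond the limit.
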SoundCue /BallisticShieldGameplay/Sounds/BallisticShield_DropImpact_Cue
root=Mixer_0
nodes [
 {id=Mixer_0, children [Modulator_0,Modulator_1], properties {inputVolume [1,1]}}
 {id=Modulator_0, children [Random_1]}
 {id=Modulator_1, children [QualityLevel_1], properties {pitchMax=1, pitchMin=1}}
 {id=QualityLevel_1, children [Random_2,Random_2,Random_3,Random_3]}
 {id=Random_1, children [WavePlayer_1,WavePlayer_2,WavePlayer_7], properties {weights [1,1,1]}}
 {
  id=Random_2
  children [WavePlayer_9, WavePlayer_10, WavePlayer_11, WavePlayer_12, WavePlayer_13, WavePlayer_14]
  properties {weights [1,1,1,1,1,1]}
 }
 {id=Random_3, children [WavePlayer_12, WavePlayer_13, WavePlayer_14], properties {bShouldExcludeFromBranchCulling=true, weights [1,1,1]}}
 {id=WavePlayer_1, properties {soundWaveAssetPtr=&/Paprika_BallisticShield_DropLand_Sweetner_04}}
 {id=WavePlayer_10, properties {soundWaveAssetPtr=&/Paprika_BallisticShield_DropLand_02}}
 {id=WavePlayer_11, properties {soundWaveAssetPtr=&/Paprika_BallisticShield_DropLand_03}}
 {id=WavePlayer_12, properties {soundWaveAssetPtr=&/Paprika_BallisticShield_DropLand_04}}
 {id=WavePlayer_13, properties {soundWaveAssetPtr=&/Paprika_BallisticShield_DropLand_05}}
 {id=WavePlayer_14, properties {soundWaveAssetPtr=&/Paprika_BallisticShield_DropLand_06}}
 {id=WavePlayer_2, properties {soundWaveAssetPtr=&/Paprika_BallisticShield_DropLand_Sweetner_03}}
 {id=WavePlayer_7, properties {soundWaveAssetPtr=&/Paprika_BallisticShield_DropLand_Sweetner_02}}
 {id=WavePlayer_9, properties {soundWaveAssetPtr=&/Paprika_BallisticShield_DropLand_01}}
]
properties {
 attenuationSettings=&Fort_Weapon_Drop_Att
 bOverrideConcurrency=true
 maxDistance=3000
 soundClassObject=&Fort_Loot_Collect
 volumeMultiplier=0.4
 attenuationOverrides {falloffDistance=1000, attenuationShapeExtents {x=0}}
 concurrencyOverrides {maxCount=3, resolutionRule=PreventNew}
}
