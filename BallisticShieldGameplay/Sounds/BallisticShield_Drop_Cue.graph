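# Sound cue for dropping the Ballistic Shield, with weapon-drop attenuation, the loot-collect sound class and stop-oldest drop concurrency.
# A mixer layers a prominent borrowed Air Strike drop, a quieter, pitched-down dual-pistol drop and randomized shield foley.
# The foley runs through a quality-level switch: the first two levels pick from five variations, the last two from a three-clip subset excluded from branch culling.
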SoundCue /BallisticShieldGameplay/Sounds/BallisticShield_Drop_Cue
root=Mixer_0
nodes [
 {id=Mixer_0, children [WavePlayer_1, Modulator_0, QualityLevel_0], properties {inputVolume [1.4,0.4,1]}}
 {id=Modulator_0, children [WavePlayer_0], properties {pitchMax=0.9, pitchMin=0.9}}
 {id=QualityLevel_0, children [Random_0,Random_0,Random_1,Random_1]}
 {id=Random_0, children [WavePlayer_2,WavePlayer_3,WavePlayer_4,WavePlayer_5,WavePlayer_6], properties {weights [1,1,1,1,1]}}
 {id=Random_1, children [WavePlayer_4,WavePlayer_5,WavePlayer_6], properties {bShouldExcludeFromBranchCulling=true, weights [1,1,1]}}
 {id=WavePlayer_0, properties {soundWaveAssetPtr=/Game/Sounds/Fort_Weapons/pistol/Dual/DualPistols_Drop_02}}
 {id=WavePlayer_1, properties {soundWaveAssetPtr=/Game/Athena/Sounds/Weapons/Air_Strike/AppleSauce_Drop_01}}
 {id=WavePlayer_2, properties {soundWaveAssetPtr=&/Paprika_BallisticShield_Foley_05}}
 {id=WavePlayer_3, properties {soundWaveAssetPtr=&/Paprika_BallisticShield_Foley_04}}
 {id=WavePlayer_4, properties {soundWaveAssetPtr=&/Paprika_BallisticShield_Foley_03}}
 {id=WavePlayer_5, properties {soundWaveAssetPtr=&/Paprika_BallisticShield_Foley_02}}
 {id=WavePlayer_6, properties {soundWaveAssetPtr=&/Paprika_BallisticShield_Foley_01}}
]
properties {
 attenuationSettings=&Fort_Weapon_Drop_Att
 maxDistance=3000
 soundClassObject=&Fort_Loot_Collect
 volumeMultiplier=1
 concurrencyOverrides {resolutionRule=PreventNew}
 concurrencySet [&Weapon_Drop_StopOldest]
}
