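# Sound cue for equipping the ballistic shield, routed as weapon foley with reload attenuation and player melee weapon concurrency.
# A mixer layers a modulated equip sound, a delayed, louder pistol-draw sound and a delayed, modulated, quieter foley sound.
# Each layer picks a random variation.
# Quality-level nodes switch the equip and foley layers between four-variation and two-variation pools, with the smaller pools excluded from branch culling.
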SoundCue /BallisticShieldGameplay/Sounds/BallisticShield_Equip_Cue
root=Mixer_0
nodes [
 {id=Delay_0, children [Random_2], properties {delayMax=0.4, delayMin=0.4}}
 {id=Delay_1, children [QualityLevel_1], properties {delayMax=0.4, delayMin=0.4}}
 {id=Mixer_0, children [Modulator_0,Delay_0,Modulator_1], properties {inputVolume [1,2,0.7]}}
 {id=Modulator_0, children [QualityLevel_0]}
 {id=Modulator_1, children [Delay_1]}
 {id=QualityLevel_0, children [Random_0,Random_0,Random_1,Random_1]}
 {id=QualityLevel_1, children [Random_3,Random_3,Random_4,Random_4]}
 {id=Random_0, children [WavePlayer_0,WavePlayer_1,WavePlayer_2,WavePlayer_4], properties {weights [1,1,1,1]}}
 {id=Random_1, children [WavePlayer_2,WavePlayer_4], properties {bShouldExcludeFromBranchCulling=true, weights [1,1]}}
 {id=Random_2, children [WavePlayer_3,WavePlayer_5], properties {weights [1,1]}}
 {id=Random_3, children [WavePlayer_6,WavePlayer_7,WavePlayer_8,WavePlayer_9], properties {weights [1,1,1,1]}}
 {id=Random_4, children [WavePlayer_8,WavePlayer_9], properties {bShouldExcludeFromBranchCulling=true, weights [1,1]}}
 {id=WavePlayer_0, properties {soundWaveAssetPtr=&/Paprika_BallisticShield_Equip_01}}
 {id=WavePlayer_1, properties {soundWaveAssetPtr=&/Paprika_BallisticShield_Equip_02}}
 {id=WavePlayer_2, properties {soundWaveAssetPtr=&/Paprika_BallisticShield_Equip_03}}
 {id=WavePlayer_3, properties {soundWaveAssetPtr=&/Paprika_BallisticShield_DrawPistol_02}}
 {id=WavePlayer_4, properties {soundWaveAssetPtr=&/Paprika_BallisticShield_Equip_04}}
 {id=WavePlayer_5, properties {soundWaveAssetPtr=&/Paprika_BallisticShield_DrawPistol_01}}
 {id=WavePlayer_6, properties {soundWaveAssetPtr=&/Paprika_BallisticShield_Foley_01}}
 {id=WavePlayer_7, properties {soundWaveAssetPtr=&/Paprika_BallisticShield_Foley_02}}
 {id=WavePlayer_8, properties {soundWaveAssetPtr=&/Paprika_BallisticShield_Foley_03}}
 {id=WavePlayer_9, properties {soundWaveAssetPtr=&/Paprika_BallisticShield_Foley_04}}
]
properties {
 attenuationSettings=&Fort_Weapon_Reload_Att
 bDebug=true
 maxDistance=2250
 soundClassObject=&Fort_WeaponFoley
 volumeMultiplier=0.501
 attenuationOverrides {distanceAlgorithm=NaturalSound, falloffDistance=3000, attenuationShapeExtents {x=1000}}
 concurrencySet [&Player_Melee_Weapon_Conc]
}
