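# Sound cue for picking up the ballistic shield, using weapon pickup attenuation and the loot-collect sound class.
# A mixer layers shield foley with a quieter pickup sound, each through a modulator for variation.
# Per layer, a quality-level node selects a random pool of five waves or a three-wave subset, the latter excluded from branch culling.
# Concurrency uses a shared pickup set, plus single-instance, prevent-new override values.
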SoundCue /BallisticShieldGameplay/Sounds/BallisticShield_Pickup_Cue
root=Mixer_0
nodes [
 {id=Mixer_0, children [Modulator_0,Modulator_1], properties {inputVolume [1,0.35]}}
 {id=Modulator_0, children [QualityLevel_0]}
 {id=Modulator_1, children [QualityLevel_1]}
 {id=QualityLevel_0, children [Random_0,Random_0,Random_1,Random_1]}
 {id=QualityLevel_1, children [Random_2,Random_2,Random_3,Random_3]}
 {id=Random_0, children [WavePlayer_0,WavePlayer_1,WavePlayer_2,WavePlayer_3,WavePlayer_4], properties {weights [1,1,1,1,1]}}
 {id=Random_1, children [WavePlayer_2,WavePlayer_3,WavePlayer_4], properties {bShouldExcludeFromBranchCulling=true, weights [1,1,1]}}
 {id=Random_2, children [WavePlayer_5, WavePlayer_6, WavePlayer_7, WavePlayer_8, WavePlayer_10], properties {weights [1,1,1,1,1]}}
 {id=Random_3, children [WavePlayer_7, WavePlayer_8, WavePlayer_10], properties {bShouldExcludeFromBranchCulling=true, weights [1,1,1]}}
 {id=WavePlayer_0, properties {soundWaveAssetPtr=&/Paprika_BallisticShield_Foley_05}}
 {id=WavePlayer_1, properties {soundWaveAssetPtr=&/Paprika_BallisticShield_Foley_04}}
 {id=WavePlayer_10, properties {soundWaveAssetPtr=&/Paprika_BallisticShield_PickUp_01}}
 {id=WavePlayer_2, properties {soundWaveAssetPtr=&/Paprika_BallisticShield_Foley_03}}
 {id=WavePlayer_3, properties {soundWaveAssetPtr=&/Paprika_BallisticShield_Foley_02}}
 {id=WavePlayer_4, properties {soundWaveAssetPtr=&/Paprika_BallisticShield_Foley_01}}
 {id=WavePlayer_5, properties {soundWaveAssetPtr=&/Paprika_BallisticShield_PickUp_05}}
 {id=WavePlayer_6, properties {soundWaveAssetPtr=&/Paprika_BallisticShield_PickUp_04}}
 {id=WavePlayer_7, properties {soundWaveAssetPtr=&/Paprika_BallisticShield_PickUp_03}}
 {id=WavePlayer_8, properties {soundWaveAssetPtr=&/Paprika_BallisticShield_PickUp_02}}
]
properties {
 attenuationSettings=&Fort_Weapon_Pickup_Att
 maxDistance=3500
 pitchMultiplier=0.9
 soundClassObject=&Fort_Loot_Collect
 volumeMultiplier=1
 concurrencyOverrides {maxCount=1, resolutionRule=PreventNew}
 concurrencySet [&PickupMaximum_Oldest]
}
