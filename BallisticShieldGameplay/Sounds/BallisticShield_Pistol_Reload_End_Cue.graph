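# Sound cue for the end of the Ballistic Shield pistol reload, played as weapon foley.
# A single wave player for the reload-end wave feeds a Modulator at the root.
# It uses the shared weapon reload attenuation and overrides concurrency to cap simultaneous instances.
# Its debug flag is left switched on.
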SoundCue /BallisticShieldGameplay/Sounds/BallisticShield_Pistol_Reload_End_Cue
root=Modulator_3
nodes [
 {id=Modulator_3, children [WavePlayer_0]}
 {id=WavePlayer_0, properties {soundWaveAssetPtr=&/Paprika_BallisticShield_Pistol_Reload_End}}
]
properties {
 attenuationSettings=&Fort_Weapon_Reload_Att
 bDebug=true
 bOverrideConcurrency=true
 maxDistance=2250
 soundClassObject=&Fort_WeaponFoley
 volumeMultiplier=1.002
 concurrencyOverrides {maxCount=3}
}
