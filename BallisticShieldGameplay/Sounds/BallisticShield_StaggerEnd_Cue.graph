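# Sound cue for the end of a Ballistic Shield stagger.
# It waits a short fixed delay, then plays one of four equally weighted max-charge one-shot waves chosen at random.
# Playback is quiet, uses the shield's attenuation and the melee weapon sound class, and a concurrency override prevents new instances when the limit is reached.
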SoundCue /BallisticShieldGameplay/Sounds/BallisticShield_StaggerEnd_Cue
root=Delay_0
nodes [
 {id=Delay_0, children [Random_4], properties {delayMax=0.1, delayMin=0.1}}
 {id=Random_4, children [WavePlayer_1,WavePlayer_2,WavePlayer_3,WavePlayer_4], properties {weights [1,1,1,1]}}
 {id=WavePlayer_1, properties {soundWaveAssetPtr=&/Paprika_BallisticShield_MaxChargeOneShot_01}}
 {id=WavePlayer_2, properties {soundWaveAssetPtr=&/Paprika_BallisticShield_MaxChargeOneShot_02}}
 {id=WavePlayer_3, properties {soundWaveAssetPtr=&/Paprika_BallisticShield_MaxChargeOneShot_03}}
 {id=WavePlayer_4, properties {soundWaveAssetPtr=&/Paprika_BallisticShield_MaxChargeOneShot_04}}
]
properties {
 attenuationSettings=&/BallisticShield_Att
 bOverrideConcurrency=true
 maxDistance=3400
 soundClassObject=&Fort_WeaponMelee
 volumeMultiplier=0.25
 concurrencyOverrides {resolutionRule=PreventNew}
}
