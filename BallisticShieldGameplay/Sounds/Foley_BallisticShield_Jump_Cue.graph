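# Sound cue for ballistic shield jump foley, playing one of fifteen equally weighted random foley waves in the melee weapon sound class.
# An "islocal" bool branch sends local playback through 2D attenuation.
# Non-local or unset playback uses the shield's standard attenuation plus a modulator that boosts volume, with both paths sharing the same random node.
# A concurrency override limits the cue to a single instance.
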SoundCue /BallisticShieldGameplay/Sounds/Foley_BallisticShield_Jump_Cue
root=Branch_0
nodes [
 {id=Attenuation_0, children [Random_3], properties {attenuationSettings=&/BallisticShield_2D_Att}}
 {id=Attenuation_1, children [Random_3], properties {attenuationSettings=&/BallisticShield_Att}}
 {id=Branch_0, children [Attenuation_0, Modulator_0, Modulator_0], properties {boolParameterName=islocal}}
 {id=Modulator_0, children [Attenuation_1], properties {pitchMax=1, volumeMax=1.2, volumeMin=1.2}}
 {
  id=Random_3
  children [
   WavePlayer_5
   WavePlayer_6
   WavePlayer_7
   WavePlayer_8
   WavePlayer_12
   WavePlayer_13
   WavePlayer_14
   WavePlayer_15
   WavePlayer_16
   WavePlayer_17
   WavePlayer_18
   WavePlayer_19
   WavePlayer_20
   WavePlayer_21
   WavePlayer_22
  ]
  properties {weights [1,1,1,1,1,1,1,1,1,1,1,1,1,1,1]}
 }
 {id=WavePlayer_12, properties {soundWaveAssetPtr=&/Paprika_BallisticShield_Foley_05}}
 {id=WavePlayer_13, properties {soundWaveAssetPtr=&/Paprika_BallisticShield_Foley_06}}
 {id=WavePlayer_14, properties {soundWaveAssetPtr=&/Paprika_BallisticShield_Foley_07}}
 {id=WavePlayer_15, properties {soundWaveAssetPtr=&/Paprika_BallisticShield_Foley_08}}
 {id=WavePlayer_16, properties {soundWaveAssetPtr=&/Paprika_BallisticShield_Foley_09}}
 {id=WavePlayer_17, properties {soundWaveAssetPtr=&/Paprika_BallisticShield_Foley_10}}
 {id=WavePlayer_18, properties {soundWaveAssetPtr=&/Paprika_BallisticShield_Foley_11}}
 {id=WavePlayer_19, properties {soundWaveAssetPtr=&/Paprika_BallisticShield_Foley_12}}
 {id=WavePlayer_20, properties {soundWaveAssetPtr=&/Paprika_BallisticShield_Foley_13}}
 {id=WavePlayer_21, properties {soundWaveAssetPtr=&/Paprika_BallisticShield_Foley_14}}
 {id=WavePlayer_22, properties {soundWaveAssetPtr=&/Paprika_BallisticShield_Foley_15}}
 {id=WavePlayer_5, properties {soundWaveAssetPtr=&/Paprika_BallisticShield_Foley_01}}
 {id=WavePlayer_6, properties {soundWaveAssetPtr=&/Paprika_BallisticShield_Foley_02}}
 {id=WavePlayer_7, properties {soundWaveAssetPtr=&/Paprika_BallisticShield_Foley_03}}
 {id=WavePlayer_8, properties {soundWaveAssetPtr=&/Paprika_BallisticShield_Foley_04}}
]
properties {
 attenuationSettings=&BallisticShield_Att
 bOverrideConcurrency=true
 maxDistance=3400
 soundClassObject=&Fort_WeaponMelee
 volumeMultiplier=0.5
 concurrencyOverrides {maxCount=1}
}
