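# Sound cue for the ballistic shield's landing foley, limited to one concurrent instance in the melee weapon sound class.
# It mixes a random equip sound at reduced level with a random, pitched-up heavy footstep sound, each from four variants.
# An `islocal` branch sends that shared mix through 2D attenuation for the local player, or through volume-boosted shield attenuation when false or unset.
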SoundCue /BallisticShieldGameplay/Sounds/Foley_BallisticShield_Land_Cue
root=Branch_0
nodes [
 {id=Attenuation_0, children [Mixer_2], properties {attenuationSettings=&/BallisticShield_2D_Att}}
 {id=Attenuation_1, children [Mixer_2], properties {attenuationSettings=&/BallisticShield_Att}}
 {id=Branch_0, children [Attenuation_0, Modulator_0, Modulator_0], properties {boolParameterName=islocal}}
 {id=Mixer_2, children [Modulator_2,Modulator_1], properties {inputVolume [0.7,1]}}
 {id=Modulator_0, children [Attenuation_1], properties {pitchMax=1, volumeMax=1.2, volumeMin=1.2}}
 {id=Modulator_1, children [Random_2], properties {pitchMax=1.2, pitchMin=1.1}}
 {id=Modulator_2, children [Random_0]}
 {id=Random_0, children [WavePlayer_0,WavePlayer_5,WavePlayer_6,WavePlayer_7], properties {weights [1,1,1,1]}}
 {id=Random_2, children [WavePlayer_36, WavePlayer_37, WavePlayer_38, WavePlayer_39], properties {weights [1,1,1,1]}}
 {id=WavePlayer_0, properties {soundWaveAssetPtr=&/Paprika_BallisticShield_Equip_01}}
 {id=WavePlayer_36, properties {soundWaveAssetPtr=&/Paprika_BallisticShield_HeavyFootsteps_17}}
 {id=WavePlayer_37, properties {soundWaveAssetPtr=&/Paprika_BallisticShield_HeavyFootsteps_16}}
 {id=WavePlayer_38, properties {soundWaveAssetPtr=&/Paprika_BallisticShield_HeavyFootsteps_03}}
 {id=WavePlayer_39, properties {soundWaveAssetPtr=&/Paprika_BallisticShield_HeavyFootsteps_01}}
 {id=WavePlayer_5, properties {soundWaveAssetPtr=&/Paprika_BallisticShield_Equip_02}}
 {id=WavePlayer_6, properties {soundWaveAssetPtr=&/Paprika_BallisticShield_Equip_03}}
 {id=WavePlayer_7, properties {soundWaveAssetPtr=&/Paprika_BallisticShield_Equip_04}}
]
properties {
 attenuationSettings=&BallisticShield_Att
 bOverrideConcurrency=true
 maxDistance=3400
 soundClassObject=&Fort_WeaponMelee
 volumeMultiplier=0.7
 concurrencyOverrides {maxCount=1}
}
