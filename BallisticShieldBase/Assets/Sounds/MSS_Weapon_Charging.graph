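# Stereo MetaSound Source for weapon charging audio, mixing four layers: charge-up one-shot and loop, max-charge one-shot and loop.
# Play starts the charge-up layers, while elapsed time mapped over ChargeLength modulates the charge loop's gain and pitch.
# A self-closing trigger gate fires max charge once, starting its one-shot and delayed loop and optionally stopping the charge-up loop.
# Global gain and pitch inputs adjust all layers.
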MetaSoundSource /BallisticShieldBase/Assets/Sounds/MSS_Weapon_Charging
inputs [
 ChargeLength {type=Float}
 ChargeUpLoop.Delay {type=Time, access=Constructor}
 "ChargeUpLoop.Gain (Db)" {type=Float}
 "ChargeUpLoop.GainModulation.Max (Db)" {type=Float}
 "ChargeUpLoop.GainModulation.Min (Db)" {type=Float}
 "ChargeUpLoop.PitchModulation.Max (Semitone)" {type=Float}
 "ChargeUpLoop.PitchModulation.Min (Semitone)" {type=Float}
 "ChargeUpLoop.PitchShift (Semitones)" {type=Float}
 ChargeUpLoop.StopOnMaxCharge {type=Bool, access=Constructor}
 ChargeUpLoop.Wave {type=WaveAsset:Array, access=Constructor}
 "ChargeUpOneshot.Gain (Db)" {type=Float}
 "ChargeUpOneshot.PitchShift (Semitones)" {type=Float}
 ChargeUpOneshot.Wave {type=WaveAsset:Array, access=Constructor}
 "Global.Gain (Db)" {type=Float}
 "Global.PitchShift (Semitones)" {type=Float}
 MaxChargeLoop.Delay {type=Time, access=Constructor}
 "MaxChargeLoop.Gain (Db)" {type=Float}
 "MaxChargeLoop.PitchShift (Semitones)" {type=Float}
 MaxChargeLoop.Wave {type=WaveAsset:Array, access=Constructor}
 "MaxChargeOneshot.Gain (Db)" {type=Float}
 "MaxChargeOneshot.PitchShift (Semitones)" {type=Float}
 MaxChargeOneshot.Wave {type=WaveAsset:Array, access=Constructor}
 Time.ElapsedTime {type=Float}
 Time.PlayTime {type=Float}
 Time.TimeSinceLastPlayed {type=Float}
 UE.Source.OnPlay {type=Trigger}
]
interfaces [Fort.Time:1.0, UE.OutputFormat.Stereo:1.0, UE.Source.OneShot:1.0, UE.Source:1.1]
outputs [
 "UE.OutputFormat.Stereo.Audio:0" {type=Audio}
 "UE.OutputFormat.Stereo.Audio:1" {type=Audio}
 UE.Source.OneShot.OnFinished {type=Trigger}
]
pages [Default {
  input_defaults {
   ChargeLength=2
   ChargeUpLoop.Delay=0
   "ChargeUpLoop.Gain (Db)"=0
   "ChargeUpLoop.GainModulation.Max (Db)"=0
   "ChargeUpLoop.GainModulation.Min (Db)"=0
   "ChargeUpLoop.PitchModulation.Max (Semitone)"=0
   "ChargeUpLoop.PitchModulation.Min (Semitone)"=0
   "ChargeUpLoop.PitchShift (Semitones)"=0
   ChargeUpLoop.StopOnMaxCharge="<MetaSoundBuilderResult.SUCCEEDED: 0>"
   "ChargeUpOneshot.Gain (Db)"=0
   "ChargeUpOneshot.PitchShift (Semitones)"=0
   "Global.Gain (Db)"=0
   "Global.PitchShift (Semitones)"=0
   MaxChargeLoop.Delay=0
   "MaxChargeLoop.Gain (Db)"=0
   "MaxChargeLoop.PitchShift (Semitones)"=0
   "MaxChargeOneshot.Gain (Db)"=0
   "MaxChargeOneshot.PitchShift (Semitones)"=0
   Time.ElapsedTime=0
   Time.PlayTime=0
   Time.TimeSinceLastPlayed=0
   UE.Source.OnPlay=null
   ChargeUpLoop.Wave []
   ChargeUpOneshot.Wave []
   MaxChargeLoop.Wave []
   MaxChargeOneshot.Wave []
  }
  nodes [
   DecibelstoLinearGain_0 = UE.Decibels to Linear Gain:Float(Decibels=^"ChargeUpLoop.GainModulation.Max (Db)")
   DecibelstoLinearGain_2 = UE.Decibels to Linear Gain:Float(Decibels=^"ChargeUpLoop.GainModulation.Min (Db)")
   n1 = VariableAccessor.Float(Variable=VariableMutator.Float(Variable=InitVariable.Float().Variable, Value=MapRange.MapRange:Float(In Range B=^ChargeLength, In=^Time.ElapsedTime).Out Value).Variable)
   n2 = VariableAccessor.Float(Variable=@n1.Variable)
   LinearGaintoDecibels_0 = UE.Linear Gain to Decibels:Float(Linear Gain=/FortMetasound/SoundCue/MSP_ContinuousModulator(
     Output.Max=@DecibelstoLinearGain_0.Linear Gain
     Output.Min=@DecibelstoLinearGain_2.Linear Gain
     Value In=VariableAccessor.Float(Variable=@n2.Variable).Value
    ).Value Out)
   DecibelstoLinearGain_1 = UE.Decibels to Linear Gain:Float(Decibels=UE.Add:Float(AdditionalOperands=^"ChargeUpLoop.Gain (Db)", PrimaryOperand=@LinearGaintoDecibels_0.Decibels).Out)
   TriggerOnValueChange_0 = UE.Trigger On Value Change:Float(Value=@n1.Value)
   n3 = VariableDeferredAccessor.Trigger(Variable=InitVariable.Trigger().Variable)
   TriggerControl_0 = UE.Trigger Control(
    Trigger In=TriggerCompare.Clamp:Float(A=UE.Reroute(Value=@n1.Value).Value, Compare=@TriggerOnValueChange_0.Trigger) {also {defaults {B=1, Type=5}}}.True
    Close=@n3.Value
   )
   n4 = VariableAccessor.Trigger(Variable=VariableMutator.Trigger(Value=@TriggerControl_0.Trigger Out, Variable=@n3.Variable).Variable)
   n5 = /FortMetasound/Patches/MSP_Audio_WavePlayer_Basic_Stereo(
    Gain in Decibels=^"MaxChargeOneshot.Gain (Db)"
    Wave Asset=^MaxChargeOneshot.Wave
    Pitch Shift=UE.Add:Float(AdditionalOperands=^"Global.PitchShift (Semitones)", PrimaryOperand=^"MaxChargeOneshot.PitchShift (Semitones)").Out
    Play Trigger=@n4.Value
   )
   n6 = &MSP_Audio_WavePlayer_Basic_Stereo(
    Gain in Decibels=^"ChargeUpOneshot.Gain (Db)"
    Wave Asset=^ChargeUpOneshot.Wave
    Play Trigger=^UE.Source.OnPlay
    Pitch Shift=UE.Add:Float(PrimaryOperand=^"ChargeUpOneshot.PitchShift (Semitones)", AdditionalOperands=^"Global.PitchShift (Semitones)").Out
   )
   n7 = VariableAccessor.Trigger(Variable=@n4.Variable)
   TriggerDelay_1 = UE.Trigger Delay(Delay Time=^MaxChargeLoop.Delay, In=@n7.Value)
   n8 = &MSP_Audio_WavePlayer_Basic_Stereo(
    Gain in Decibels=^"MaxChargeLoop.Gain (Db)"
    Wave Asset=^MaxChargeLoop.Wave
    Pitch Shift=UE.Add:Float(AdditionalOperands=^"Global.PitchShift (Semitones)", PrimaryOperand=^"MaxChargeLoop.PitchShift (Semitones)").Out
    Play Trigger=@TriggerDelay_1.Out
   ) {also {defaults {Is Looping?="<MetaSoundBuilderResult.SUCCEEDED: 0>"}}}
   TriggerDelay_0 = UE.Trigger Delay(Delay Time=^ChargeUpLoop.Delay, In=^UE.Source.OnPlay)
   ArrayRandomGet_0 = Array.Random Get:WaveAsset:Array(In Array=^ChargeUpLoop.Wave, Next=@TriggerDelay_0.Out)
   n9 = UE.Add:Float(
    PrimaryOperand=^"Global.PitchShift (Semitones)"
    AdditionalOperands=&MSP_ContinuousModulator(Output.Max=^"ChargeUpLoop.PitchModulation.Max (Semitone)", Output.Min=^"ChargeUpLoop.PitchModulation.Min (Semitone)", Value In=@n2.Value).Value Out
   )
   WavePlayer_0 = UE.Wave Player:Stereo(
    Play=@ArrayRandomGet_0.On Next
    Wave Asset=@ArrayRandomGet_0.Value
    Pitch Shift=UE.Reroute(Value=UE.Add:Float(AdditionalOperands=^"ChargeUpLoop.PitchShift (Semitones)", PrimaryOperand=@n9.Out).Out).Value
    Stop=TriggerCompare.Clamp:Bool(A=^ChargeUpLoop.StopOnMaxCharge, Compare=VariableAccessor.Trigger(Variable=@n7.Variable).Value) {also {defaults {B="<MetaSoundBuilderResult.SUCCEEDED: 0>"}}}.True
   ) {also {defaults {Loop="<MetaSoundBuilderResult.SUCCEEDED: 0>"}}}
   AudioMixerAudioMixerStereo4_0 = "AudioMixer.Audio Mixer (Stereo, 4)"(
    Gain 1=@DecibelstoLinearGain_1.Linear Gain
    In 2 L=@n5.Out L
    In 2 R=@n5.Out R
    In 0 L=@n6.Out L
    In 0 R=@n6.Out R
    In 3 L=@n8.Out L
    In 3 R=@n8.Out R
    In 1 L=@WavePlayer_0.Out Left
    In 1 R=@WavePlayer_0.Out Right
   )
   DecibelstoLinearGain_3 = UE.Decibels to Linear Gain:Float(Decibels=^"Global.Gain (Db)")
   Multiply_0 = UE.Multiply:Audio by Float(PrimaryOperand=@AudioMixerAudioMixerStereo4_0.Out R, AdditionalOperands=@DecibelstoLinearGain_3.Linear Gain)
   Multiply_1 = UE.Multiply:Audio by Float(PrimaryOperand=@AudioMixerAudioMixerStereo4_0.Out L, AdditionalOperands=@DecibelstoLinearGain_3.Linear Gain)
  ]
  links [{dst="@UE.OutputFormat.Stereo.Audio:1", src=Multiply_0.Out}, {dst="@UE.OutputFormat.Stereo.Audio:0", src=Multiply_1.Out}]
 }]
properties {
 assetImportData=&self:AssetImportData
 numChannels=2
 outputFormat=Stereo
 sampleRate=48000
 soundClassObject=&Fort_WeaponMelee_PSM
}
