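# Stereo MetaSound source for the Ballistic Shield's charging sound, with a charge wind wave as the default max-charge loop.
# It wraps the shared MSS_Weapon_Charging graph, forwarding charge length, wave, gain, pitch and modulation inputs.
# On play, an isFriendly comparison selects either a quieter, muffled treatment or a full-level, open one.
# One-pole low-pass filters and mixer gain apply it, while the inner graph drives OnFinished.
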MetaSoundSource /BallisticShieldGameplay/Sounds/MSS_BallisticShield_Charging
inputs [
 ChargeLength {type=Float}
 ChargeUpLoop.Delay {type=Time, access=Constructor}
 "ChargeUpLoop.Gain (Db)" {type=Float}
 "ChargeUpLoop.GainModulation.Max (Db)" {type=Float}
 "ChargeUpLoop.GainModulation.Min (Db)" {type=Float}
 "ChargeUpLoop.PitchModulation.Max (Semitone)" {type=Float}
 "ChargeUpLoop.PitchModulation.Min (Semitone)" {type=Float}
 "ChargeUpLoop.PitchShift (Semitones)" {type=Float}
 ChargeUpLoop.StopOnMaxCharge {type=Bool, access=Constructor}
 ChargeUpLoop.Wave {type=WaveAsset:Array, access=Constructor}
 "Global.Gain (Db)" {type=Float}
 "Global.PitchShift (Semitones)" {type=Float}
 isFriendly {type=Bool}
 MaxChargeLoop.Delay {type=Time, access=Constructor}
 "MaxChargeLoop.Gain (Db)" {type=Float}
 "MaxChargeLoop.PitchShift (Semitones)" {type=Float}
 MaxChargeLoop.Wave {type=WaveAsset:Array, access=Constructor}
 "MaxChargeOneshot.Gain (Db)" {type=Float}
 "MaxChargeOneshot.PitchShift (Semitones)" {type=Float}
 MaxChargeOneshot.Wave {type=WaveAsset:Array, access=Constructor}
 Time.ElapsedTime {type=Float}
 Time.PlayTime {type=Float}
 Time.TimeSinceLastPlayed {type=Float}
 UE.Source.OnPlay {type=Trigger}
]
interfaces [Fort.Time:1.0, UE.OutputFormat.Stereo:1.0, UE.Source.OneShot:1.0, UE.Source:1.1]
outputs [
 "UE.OutputFormat.Stereo.Audio:0" {type=Audio}
 "UE.OutputFormat.Stereo.Audio:1" {type=Audio}
 UE.Source.OneShot.OnFinished {type=Trigger}
]
pages [Default {
  input_defaults {
   ChargeLength=0.01
   ChargeUpLoop.Delay=0
   "ChargeUpLoop.Gain (Db)"=-14.128
   "ChargeUpLoop.GainModulation.Max (Db)"=0
   "ChargeUpLoop.GainModulation.Min (Db)"=0
   "ChargeUpLoop.PitchModulation.Max (Semitone)"=0
   "ChargeUpLoop.PitchModulation.Min (Semitone)"=-1
   "ChargeUpLoop.PitchShift (Semitones)"=-3
   ChargeUpLoop.StopOnMaxCharge=null
   "Global.Gain (Db)"=1
   "Global.PitchShift (Semitones)"=0
   MaxChargeLoop.Delay=0
   "MaxChargeLoop.Gain (Db)"=0.3
   "MaxChargeLoop.PitchShift (Semitones)"=0
   "MaxChargeOneshot.Gain (Db)"=-6
   "MaxChargeOneshot.PitchShift (Semitones)"=0
   Time.ElapsedTime=0
   Time.PlayTime=0
   Time.TimeSinceLastPlayed=0
   UE.Source.OnPlay=null
   isFriendly=null
   ChargeUpLoop.Wave []
   MaxChargeLoop.Wave [&/Paprika_BallisticShield_ChargeWind_Quad]
   MaxChargeOneshot.Wave []
  }
  nodes [
   n1 = TriggerCompare.Clamp:Bool(Compare=^UE.Source.OnPlay, A=^isFriendly) {also {defaults {B="<MetaSoundBuilderResult.SUCCEEDED: 0>"}}}
   TriggerRouteTriggerRouteFloat2_1 = "TriggerRoute.Trigger Route (Float, 2):Float"(Set 1=@n1.FALSE, Set 0=@n1.TRUE) {also {defaults {Value 0=-6}}}
   DecibelstoLinearGain_0 = UE.Decibels to Linear Gain:Float(Decibels=@TriggerRouteTriggerRouteFloat2_1.Value)
   MSS_Weapon_Charging_0 = /BallisticShieldBase/Assets/Sounds/MSS_Weapon_Charging(
    ChargeLength=^ChargeLength
    ChargeUpLoop.Delay=^ChargeUpLoop.Delay
    "ChargeUpLoop.Gain (Db)"=^"ChargeUpLoop.Gain (Db)"
    "ChargeUpLoop.GainModulation.Max (Db)"=^"ChargeUpLoop.GainModulation.Max (Db)"
    "ChargeUpLoop.GainModulation.Min (Db)"=^"ChargeUpLoop.GainModulation.Min (Db)"
    "ChargeUpLoop.PitchModulation.Max (Semitone)"=^"ChargeUpLoop.PitchModulation.Max (Semitone)"
    "ChargeUpLoop.PitchModulation.Min (Semitone)"=^"ChargeUpLoop.PitchModulation.Min (Semitone)"
    "ChargeUpLoop.PitchShift (Semitones)"=^"ChargeUpLoop.PitchShift (Semitones)"
    ChargeUpLoop.StopOnMaxCharge=^ChargeUpLoop.StopOnMaxCharge
    ChargeUpLoop.Wave=^ChargeUpLoop.Wave
    "Global.Gain (Db)"=^"Global.Gain (Db)"
    "Global.PitchShift (Semitones)"=^"Global.PitchShift (Semitones)"
    MaxChargeLoop.Delay=^MaxChargeLoop.Delay
    "MaxChargeLoop.Gain (Db)"=^"MaxChargeLoop.Gain (Db)"
    "MaxChargeLoop.PitchShift (Semitones)"=^"MaxChargeLoop.PitchShift (Semitones)"
    MaxChargeLoop.Wave=^MaxChargeLoop.Wave
    "MaxChargeOneshot.Gain (Db)"=^"MaxChargeOneshot.Gain (Db)"
    "MaxChargeOneshot.PitchShift (Semitones)"=^"MaxChargeOneshot.PitchShift (Semitones)"
    MaxChargeOneshot.Wave=^MaxChargeOneshot.Wave
    Time.ElapsedTime=^Time.ElapsedTime
    Time.PlayTime=^Time.PlayTime
    Time.TimeSinceLastPlayed=^Time.TimeSinceLastPlayed
    UE.Source.OnPlay=^UE.Source.OnPlay
   )
   TriggerRouteTriggerRouteFloat2_0 = "TriggerRoute.Trigger Route (Float, 2):Float"(Set 1=@n1.FALSE, Set 0=@n1.TRUE) {also {defaults {Value 0=1400, Value 1=20000}}}
   n2 = UE.Reroute(Value=@TriggerRouteTriggerRouteFloat2_0.Value)
   OnePoleLowPassFilter_0 = UE.One-Pole Low Pass Filter:Audio(In=@MSS_Weapon_Charging_0."UE.OutputFormat.Stereo.Audio:0", Cutoff Frequency=@n2.Value)
   OnePoleLowPassFilter_1 = UE.One-Pole Low Pass Filter:Audio(In=@MSS_Weapon_Charging_0."UE.OutputFormat.Stereo.Audio:1", Cutoff Frequency=@n2.Value)
   AudioMixerAudioMixerStereo2_0 = "AudioMixer.Audio Mixer (Stereo, 2)"(Gain 0=@DecibelstoLinearGain_0.Linear Gain, In 0 L=@OnePoleLowPassFilter_0.out, In 0 R=@OnePoleLowPassFilter_1.out)
  ]
  links [
   {dst="@UE.OutputFormat.Stereo.Audio:0", src=AudioMixerAudioMixerStereo2_0.Out L}
   {dst="@UE.OutputFormat.Stereo.Audio:1", src=AudioMixerAudioMixerStereo2_0.Out R}
   {dst="@UE.Source.OneShot.OnFinished", src=MSS_Weapon_Charging_0.UE.Source.OneShot.OnFinished}
  ]
 }]
properties {
 assetImportData=&self:AssetImportData
 attenuationSettings=&/BallisticShield_Att
 maxDistance=3400
 numChannels=2
 outputFormat=Stereo
 sampleRate=48000
 soundClassObject=&Fort_WeaponMelee
}
